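# Physics asset for the female medium Vampire Casual backpack cosmetic, giving its dangling straps collision bodies and joint limits.
# Eight dyn_strap bones (top and bottom, left and right, two sets) each carry one capsule, while the pack body uses two overlapping capsules.
# Each strap is constrained directly to the pack rather than chained, with zero twist and narrow, mostly asymmetric swing limits that favour motion on one axis.
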PhysicsAsset /BRCosmetics/Accessories/FORT_Backpacks/F_MED_Vampire_Casual_Backpack/Meshes/F_MED_Vampire_Casual_Pack_Physics
bodies {
 dyn_strap01_bottom_l [{shapeType=Capsule, length=2.0769, radius=1.9231, shapeName=dyn_strap01_bottom_r_capsule, rotation=<rot 0 0 -180>}]
 dyn_strap01_bottom_r [{shapeType=Capsule, length=2.0769, radius=1.9231, shapeName=dyn_strap01_bottom_r_capsule}]
 dyn_strap01_top_l [{shapeType=Capsule, length=4.8435, radius=2.0355, shapeName=dyn_strap01_top_l_capsule}]
 dyn_strap01_top_r [{shapeType=Capsule, length=3.9633, radius=2.3983, shapeName=dyn_strap01_top_r_capsule}]
 dyn_strap02_bottom_l [{shapeType=Capsule, length=2.0572, radius=1.9428, shapeName=dyn_strap02_bottom_l_capsule}]
 dyn_strap02_bottom_r [{shapeType=Capsule, length=2.0572, radius=1.9428, shapeName=dyn_strap02_bottom_r_capsule}]
 dyn_strap02_top_l [{shapeType=Capsule, length=2.757, radius=1.243, shapeName=dyn_strap02_top_l_capsule}]
 dyn_strap02_top_r [{shapeType=Capsule, length=2.757, radius=1.243, shapeName=dyn_strap02_top_r_capsule}]
 pack [
  {shapeType=Capsule, length=0, radius=9.4111, shapeName=pack_capsule, center=<0 -1.5646 -5.3666>}
  {shapeType=Capsule, length=4.8607, radius=7.8796, shapeName=pack_capsule, center=<0 -3.933 2.4601>}
 ]
}
constraints [
 {bone1Name=dyn_strap01_bottom_r, bone2Name=pack, swing1LimitDegrees=5, swing2LimitDegrees=25, twistLimitDegrees=0}
 {bone1Name=dyn_strap01_top_l, bone2Name=pack, swing1LimitDegrees=5, swing2LimitDegrees=25, twistLimitDegrees=0}
 {bone1Name=dyn_strap01_top_r, bone2Name=pack, swing1LimitDegrees=5, swing2LimitDegrees=35, twistLimitDegrees=0}
 {bone1Name=dyn_strap02_bottom_l, bone2Name=pack, swing1LimitDegrees=5, swing2LimitDegrees=5, twistLimitDegrees=0}
 {bone1Name=dyn_strap02_bottom_r, bone2Name=pack, swing1LimitDegrees=5, swing2LimitDegrees=10, twistLimitDegrees=0}
 {bone1Name=dyn_strap02_top_l, bone2Name=pack, swing1LimitDegrees=5, swing2LimitDegrees=25, twistLimitDegrees=0}
 {bone1Name=dyn_strap02_top_r, bone2Name=pack, swing1LimitDegrees=5, swing2LimitDegrees=25, twistLimitDegrees=0}
 {bone1Name=dyn_strap01_bottom_l, bone2Name=pack, swing1LimitDegrees=5, swing2LimitDegrees=25, twistLimitDegrees=0}
]
