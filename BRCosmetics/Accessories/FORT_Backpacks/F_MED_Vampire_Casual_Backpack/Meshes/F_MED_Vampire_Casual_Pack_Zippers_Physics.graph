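# Physics asset for the zipper mesh of the Vampire Casual backpack cosmetic, defining collision and joint limits for its two dangling zipper pulls.
# A large capsule covers the pack bone, and a small capsule sits on each of the left and right dyn_zipper bones.
# Each zipper is constrained to the pack with limited swing and tight twist, with slightly different swing limits per side.
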PhysicsAsset /BRCosmetics/Accessories/FORT_Backpacks/F_MED_Vampire_Casual_Backpack/Meshes/F_MED_Vampire_Casual_Pack_Zippers_Physics
bodies {
 dyn_zipper_l [{
   shapeType=Capsule
   length=1.6855
   radius=0.7237
   shapeName=dyn_zipper_l_capsule
   center=<-1.1582 -0.0347 0.2527>
   rotation=<rot 0 -88.3085 -77.2127>
  }]
 dyn_zipper_r [{
   shapeType=Capsule
   length=1.6851
   radius=0.7327
   shapeName=dyn_zipper_r_capsule
   center=<1.2082 0.0039 0.0222>
   rotation=<rot 0 -90.4092 -89.5333>
  }]
 pack [{
   shapeType=Capsule
   length=14.5391
   radius=11.7334
   shapeName=pack_capsule
   center=<0.0538 -0.0585 -7.0233>
   rotation=<rot 0.2028 -89.8309 -0.5029>
  }]
}
constraints [
 {bone1Name=dyn_zipper_l, bone2Name=pack, swing1LimitDegrees=50, swing2LimitDegrees=30, twistLimitDegrees=10}
 {bone1Name=dyn_zipper_r, bone2Name=pack, swing1LimitDegrees=45, swing2LimitDegrees=45, twistLimitDegrees=10}
]
